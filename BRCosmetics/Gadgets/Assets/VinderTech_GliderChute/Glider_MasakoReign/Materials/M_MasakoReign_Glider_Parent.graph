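# Parent masked character-style material for the Masako Reign glider, wrapping the 2023 character PreFX and PostFX functions around custom whisker and summon stages.
# An isWhiskers switch parameter blends in SplineThicken world position offset and a flattened tube normal.
# The result passes through MF_summon, borrowed from the Headset glider, then whiskers also get an opacity mask cut from a UV step and tiled sphere mask.
# A constant temporal responsiveness output is also set.
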Material /BRCosmetics/Gadgets/Assets/VinderTech_GliderChute/Glider_MasakoReign/Materials/M_MasakoReign_Glider_Parent
graph {
 properties {
  bAutomaticallySetUsageInEditor=false
  bUseMaterialAttributes=true
  bUsedWithMeshDeformer=true
  bUsedWithMorphTargets=true
  bUsedWithSkeletalMesh=true
  blendMode=BLEND_Masked
  pixelDepthOffsetMode=PDOM_Legacy
  previewMesh=/BRCosmetics/Accessories/FORT_Backpacks/Backpack_M_MED_VigorKnackDill/Meshes/M_MED_VigorKnackDill_Pack
  shadingModel=MSM_FromMaterialExpression
  usedShadingModels=Default Lit | Subsurface | Clear Coat | Subsurface Profile | Cloth
 }
 nodes [
  n1 = TextureCoordinate()
  AppendVector(A=ComponentMask(None=@n1) {g=true}, B=ComponentMask(None=@n1) {r=true})
  n2 = MaterialFunctionCall() {materialFunction=&MF_CharacterParent_2023_PreFX}
  n3 = GetMaterialAttributes(None=@n2.Result) {attributeGetTypes [
    #1
    #2
    #3
    #4
    #5
   ]}
  n4 = TextureSample(UVs=@n1) {samplerType=SAMPLERTYPE_Normal, texture=/Game/Animation/Props/Kite/Textures/Tube_Normal_Tube_N}
  n5 = ScalarParameter() {defaultValue=-1.5, group=SplineThicken, parameterName=ST_TubeNormalFlatness, sortPriority=33}
  n6 = ScalarParameter() {controlType=Enumeration, enumeration=/Game/Global/Materials/Enums/E_Mat_OnOff, group=SplineThicken, parameterName=isWhiskers}
  n7 = ScalarParameter() {defaultValue=1, group=SplineThicken, parameterName=ST_WidthBase}
  n8 = ScalarParameter() {defaultValue=1, group=SplineThicken, parameterName=ST_WidthTip}
  n9 = StaticBoolParameter() {defaultValue=true, group=SplineThicken, parameterName=ST_AngleCorrectedNormals}
  n10 = ScalarParameter() {controlType=Enumeration, enumeration=&E_Mat_OnOff, group=SplineThicken, parameterName=isWhiskers}
  n11 = SetMaterialAttributes!(
   MaterialAttributes=@n3.MaterialAttributes
   Base Color=@n3.MaterialAttributes
   Normal=LinearInterpolate(
    A=@n3.Normal
    B=MaterialFunctionCall(Normal=@n4.RGB, Flatness=@n5) {materialFunction=&FlattenNormal}.Result
    Alpha=@n6
   )
   Metallic=@n3.MaterialAttributes
   Roughness=@n3.MaterialAttributes
   World Position Offset=LinearInterpolate(
    A=@n3.World Position Offset
    B=MaterialFunctionCall(WidthBase=@n7, WidthTip=@n8, AngleCorrectedNormals=@n9) {materialFunction=&SplineThicken}.WorldPositionOffset
    Alpha=@n10
   )
  ) {attributeSetTypes [
    #1
    #2
    #3
    #4
    #5
   ]}
  n12 = GetMaterialAttributes(None=MaterialFunctionCall!(PreskinLocalPosition=@n2.Pre skinned Local Position, Material=@n11) {materialFunction=/BRCosmetics/Gadgets/Assets/VinderTech_GliderChute/Glider_Headset/Materials/MF_summon}.Result) {attributeGetTypes [#6]}
  n13 = TextureCoordinate()
  n14 = ComponentMask(None=@n13) {g=true}
  n15 = ComponentMask(None=@n13) {r=true}
  n16 = Step(X=ComponentMask(None=AppendVector(A=@n14, B=@n15)) {r=true}) {constX=0.85, constY=0.05}
  n17 = SphereMask(A=Add(A=TextureCoordinate() {vTiling=8}, B=Constant2Vector()), B=Constant() {r=0.5}, Radius=Constant() {r=0.5})
  n18 = Multiply(A=@n12.Opacity Mask, B=Saturate(None=Add(A=@n16, B=@n17)))
  n19 = ScalarParameter() {controlType=Enumeration, enumeration=&E_Mat_OnOff, group=SplineThicken, parameterName=isWhiskers}
  MaterialFunctionCall(
   Material=SetMaterialAttributes!(MaterialAttributes=@n12.MaterialAttributes, Opacity Mask=LinearInterpolate(A=@n12.Opacity Mask, B=@n18, Alpha=@n19)) {attributeSetTypes [#6]}
   M=@n2.M
   Pre Skinned Local Position=@n2.M
   Distance Blend=@n2.M
   Fresnel=@n2.M
  ) {materialFunction=&MF_CharacterParent_2023_PostFX}
  TemporalResponsivenessOutput!(None=Constant() {r=0.5})
 ]
}
